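# Post-process-style face accessory AnimBP for the TeriyakiFish VR skin, copying the attached parent mesh's pose and adding secondary motion.
# AnimDynamics springs the top knot and cable, then constraints drive chin straps from head and torso bones and VR eyes from eye rotations.
# The cable base gets an additive rotation while skydive-floating.
# While targeting, the VR set bone's scale is replaced by a slightly reduced head world scale.
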AnimBlueprint /BRCosmetics/Characters/Player/Male/Medium/Bodies/M_MED_TeriyakiFish/Skin/Fish_VR/Meshes/M_MED_TeriyakiFish_VR_FaceAcc_AnimBP
target_skeleton=/Game/Characters/Player/Male/Medium/Bodies/M_MED_TeriyakiFish/Skin/Fish_VR/Meshes/M_MED_Teriyaki_Fish_VR_FaceAcc_Cable_usb_LOD0_Skeleton
graphs [
 AnimGraph {nodes [
   n1 = Anim Dynamics - Bone: dyn_topKnot(ComponentPose=Local To Component(LocalPose=Copy Pose From Mesh() {bUseAttachedParent=true}), Alpha=1, ExternalForce="0,0,0") {
    angularDampingOverride=1
    angularSpringConstant=2000
    bAngularSpring=true
    bGravityOverrideInSimSpace=true
    bOverrideAngularDamping=true
    simulationSpace=RootRelative
    alphaBoolBlend {blendOption=Linear}
    boundBone {boneName=dyn_topKnot}
    componentAppliedLinearAccClamp=<5000 5000 5000>
    componentLinearAccScale=<1 1 1>
    componentLinearVelScale=<0.5 0.5 0.5>
    gravityOverride {z=-400}
    physicsBodyDefinitions [{
      sphereCollisionRadius=0
      boundBone {boneName=dyn_topKnot}
      boxExtents=<10 10 10>
      constraintSetup {angularLimitsMax=<10 10 12>, angularLimitsMin=<-10 -10 -12>, angularTarget=<1 0 0>}
      localJointOffset=<-1 0 0>
     }]
    relativeSpaceBone {boneName=head}
   }
   n2 = Anim Dynamics - Bone: dyn_cable(ComponentPose=@n1, Alpha=1, ExternalForce="0,0,0") {
    angularDampingOverride=1
    angularSpringConstant=2000
    bAngularSpring=true
    bGravityOverrideInSimSpace=true
    bOverrideAngularDamping=true
    simulationSpace=RootRelative
    alphaBoolBlend {blendOption=Linear}
    boundBone {boneName=dyn_cable}
    componentAppliedLinearAccClamp=<5000 5000 5000>
    componentLinearAccScale=<0.85 0.85 0.85>
    componentLinearVelScale=<0.55 0.55 0.55>
    gravityOverride {z=-400}
    physicsBodyDefinitions [{
      sphereCollisionRadius=0
      boundBone {boneName=dyn_cable}
      boxExtents=<18 8 18>
      constraintSetup {angularLimitsMax=<15 15 15>, angularLimitsMin=<-15 -15 -15>, angularTarget=<1 0 0>}
      localJointOffset=<10 8 0>
     }]
    relativeSpaceBone {boneName=head}
   }
   n3 = "Constraint - dyn_chinStrap_l (PPPPP)"!(
    ComponentPose=@n2
    Alpha=1
    ConstraintWeights_0=0.04
    ConstraintWeights_1:float=0.04
    ConstraintWeights_2:float=0.86
    ConstraintWeights_3:float=0.02
    ConstraintWeights_4:float=0.04
   ) -> (Pose) {
    alphaBoolBlend {blendOption=Linear}
    boneToModify {boneName=dyn_chinStrap_l}
    constraintSetup [
     {targetBone {boneName=clavicle_l}}
     {targetBone {boneName=clavicle_r}}
     {targetBone {boneName=head}}
     {targetBone {boneName=spine_04}}
     {targetBone {boneName=spine_05}}
    ]
    constraintWeights [1,1,1,1,1]
   }
   n4 = "Constraint - dyn_chinStrap_r (PPPPP)"!(
    ComponentPose=@n3
    Alpha=1
    ConstraintWeights_0=0.04
    ConstraintWeights_1:float=0.04
    ConstraintWeights_2:float=0.86
    ConstraintWeights_3:float=0.02
    ConstraintWeights_4:float=0.04
   ) -> (Pose) {
    alphaBoolBlend {blendOption=Linear}
    boneToModify {boneName=dyn_chinStrap_r}
    constraintSetup [
     {targetBone {boneName=clavicle_l}}
     {targetBone {boneName=clavicle_r}}
     {targetBone {boneName=head}}
     {targetBone {boneName=spine_04}}
     {targetBone {boneName=spine_05}}
    ]
    constraintWeights [1,1,1,1,1]
   }
   n5 = "Constraint - dyn_vrEye_l (R)"(ComponentPose=@n4, Alpha=1, ConstraintWeights_0=1) {
    alphaBoolBlend {blendOption=Linear}
    boneToModify {boneName=dyn_vrEye_l}
    constraintSetup [{transformType=Rotation, targetBone {boneName=L_eye}}]
    constraintWeights [1]
   }
   n6 = "Constraint - dyn_vrEye_r (R)"(ComponentPose=@n5, Alpha=1, ConstraintWeights_0=1) {
    alphaBoolBlend {blendOption=Linear}
    boneToModify {boneName=dyn_vrEye_r}
    constraintSetup [{transformType=Rotation, targetBone {boneName=R_eye}}]
    constraintWeights [1]
   }
   n7 = "Transform (Modify) Bone - Bone: dyn_cable_base"(ComponentPose=@n6, Alpha=1, Translation="0,0,0", Rotation=Get Shoulder_diving_up_l() -> (Shoulder_diving_up_l:Rotator), Scale="1,1,1") {
    rotationMode=BMM_Additive
    rotationSpace=BCS_BoneSpace
    scaleMode=BMM_Additive
    scaleSpace=BCS_BoneSpace
    translationMode=BMM_Additive
    translationSpace=BCS_BoneSpace
    alphaBoolBlend {blendOption=Linear}
    boneToModify {boneName=dyn_cable_base}
   }
   n8 = SelectFloat(A=1, B=0, bPickA=Get bIsTargeting())
   n9 = Get VR Scale() -> (VR Scale:Vector)
   Output Pose(Result=Component To Local(ComponentPose="Transform (Modify) Bone - Bone: const_vr_set"(ComponentPose=@n7, Alpha=@n8, Translation="0,0,0", Rotation="0,0,0", Scale=@n9) {scaleMode=BMM_Replace, alphaBoolBlend {blendOption=Linear}, boneToModify {boneName=const_vr_set}})) {layerGroup=None}
  ]}
 EventGraph {nodes [
   Event BlueprintUpdateAnimation()
   > n1 = Sequence()
   n2 = GetBoneTransform!(
    self:Skinned Mesh Component Object Reference=GetOwnerSkeletalMeshComponent()
    InBoneName:Name=head
    TransformSpace:ERelativeTransformSpace Enum=RTS_World
   ) -> (ReturnValue:Transform)
   > Set VR Scale(VR Scale:Vector=vector * vector(A=BreakTransform(InTransform=@n2).Scale, B:Vector="0.95,0.91,0.91") -> (ReturnValue=~)) -> (Output_Get:Vector)
   n3 = Branch(execute=@n1.then_1, Condition=Get bIsSkydiveFloating())
   > Set Shoulder_diving_up_l(Shoulder_diving_up_l:Rotator="0,-60,20") -> (Output_Get:Rotator)
   Set Shoulder_diving_up_l(execute=@n3.else, Shoulder_diving_up_l="0, 0, 0")
   TryGetPawnOwner()
   Set Shoulder_rot_up_l(Shoulder_rot_up_l:double=GetAlphaFromReferenceTransformDelta(
     SkeletalMeshComponent=Get Skel_Mesh()
     BoneName=clavicle_l
     BaseName=spine_05
     DeltaComponentType=EulerY
     InputRange=(X=-15,Y=150)
     TwistAxis="0,0,1"
    )) -> (Output_Get:double)
  ]}
]
variables [
 Shoulder_diving_up_l {type=Rotator}
 Shoulder_rot_up_l {type=number}
 Skel_Mesh {type=SkeletalMeshComponent}
 VR Scale {type=Vector}
]
